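# Unlit translucent debug material for Niagara sprites in Niagara Fluids, built to visualize a texture supplied through a parameter.
# It samples the Tex texture parameter with standard UVs and halves its blue channel.
# A separate Opacity scalar parameter defaults to fully opaque.
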Material /NiagaraFluids/Materials/Utility/Debug
graph {
 properties {bUsedWithNiagaraSprites=true, blendMode=BLEND_Translucent, pixelDepthOffsetMode=PDOM_Legacy, shadingModel=MSM_Unlit, usedShadingModels=Unlit}
 nodes [
  Multiply(A=TextureSample(UVs=TextureCoordinate(), Tex=TextureObjectParameter() {parameterName=Tex}).B) {constB=0.5}
  ScalarParameter() {defaultValue=1, parameterName=Opacity}
 ]
}
