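# Debug display material for 2D gas in Niagara Fluids: an unlit, translucent material usable on Niagara sprites.
# The Tex texture object parameter is sampled at the sprite UVs, and its RGB is scaled down by a constant divide, then passed through a zero-offset subtract.
# A standalone constant and a pale blue vector parameter are also present; the graph shows no links from any node to the material outputs.
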Material /NiagaraFluids/Materials/Utility/DebugDisplayGas2D
graph {
 properties {bUsedWithNiagaraSprites=true, blendMode=BLEND_Translucent, pixelDepthOffsetMode=PDOM_Legacy, shadingModel=MSM_Unlit, usedShadingModels=Unlit}
 nodes [
  Constant() {r=1}
  n1 = TextureCoordinate()
  n2 = TextureObjectParameter() {parameterName=Tex}
  Subtract(A=Divide(A=TextureSample(UVs=@n1, Tex=@n2).RGB) {constB=200}) {constB=0}
  VectorParameter() {defaultValue=<rgba 0.7119 0.7745 0.9427 0>}
 ]
}
