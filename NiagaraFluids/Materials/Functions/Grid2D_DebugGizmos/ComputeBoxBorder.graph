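# ComputeBoxBorder is a material function in the Niagara Fluids Grid2D debug gizmo set, exposed to the material library.
# It takes a Vector2 UV input, samples the fixed Grid2D_Cell texture at that coordinate and outputs the RGB.
# Despite the name, it does no border math; any border shape comes from the texture itself.
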MaterialFunction /NiagaraFluids/Materials/Functions/Grid2D_DebugGizmos/ComputeBoxBorder
graph {
 properties {bExposeToLibrary=true}
 nodes [FunctionOutput(None=TextureSample(
    UVs=FunctionInput() {inputName=UV, inputType=FunctionInput_Vector2}
    Tex=TextureObject() {texture=/NiagaraFluids/Textures/Grid2D_DebugGizmos/Grid2D_Cell}
   ).RGB)]
}
