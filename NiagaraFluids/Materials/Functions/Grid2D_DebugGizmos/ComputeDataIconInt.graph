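# ComputeDataIconInt is a library-exposed material function in the Niagara Fluids Grid2D debug gizmos that draws a number as a small on-screen icon.
# ComputeIconUV turns the UV, Size, Offset, Scale and CropLeft inputs into a local icon UV and a mask.
# DebugScalarValues renders Number at that UV with at most four digits, and the mask multiplies the result to give the RGB Color output.
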MaterialFunction /NiagaraFluids/Materials/Functions/Grid2D_DebugGizmos/ComputeDataIconInt
graph {
 properties {bExposeToLibrary=true}
 nodes [
  n1 = FunctionInput(Preview=Constant() {r=512}) {bUsePreviewValueAsDefault=true, inputName=Number, inputType=FunctionInput_Scalar, previewValue=<8 8 8 8>}
  n2 = FunctionInput(Preview=Constant() {r=8}) {bUsePreviewValueAsDefault=true, inputName=Size, inputType=FunctionInput_Scalar, previewValue=<8 8 8 8>}
  n3 = Constant2Vector()
  n4 = FunctionInput(Preview=Constant2Vector() {g=1, r=1}) {bUsePreviewValueAsDefault=true, inputName=Scale, inputType=FunctionInput_Vector2, previewValue=<8 8 8 8>}
  n5 = FunctionInput(Preview=Constant()) {bUsePreviewValueAsDefault=true, inputName=CropLeft, inputType=FunctionInput_Scalar, previewValue=<8 8 8 8>}
  n6 = MaterialFunctionCall!(
   Size=@n2
   Offset=ComponentMask(None=FunctionInput(Preview=@n3) {bUsePreviewValueAsDefault=true, inputName=Offset, inputType=FunctionInput_Vector2}) {g=true, r=true}
   Scale=@n4
   CropLeft=@n5
   UV=FunctionInput(Preview=TextureCoordinate()) {bUsePreviewValueAsDefault=true, inputName=UV, inputType=FunctionInput_Vector2}
  ) {materialFunction=&/ComputeIconUV}
  n7 = MaterialFunctionCall(Number=@n1, MaximumNumberOfDigits=Constant() {r=4}, UVs=@n6.UV) {materialFunction=&DebugScalarValues}
  FunctionOutput(None=ComponentMask(None=Multiply(A=@n7.Result, B=@n6.Mask)) {b=true, g=true, r=true}) {outputName=Color}
 ]
}
