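# Material function that draws a positioned, scaled icon for Niagara Fluids Grid2D debug gizmos, using the Grid2D nametag texture by default.
# ComputeIconUV turns the UV, Size, Offset and Scale inputs into icon UVs and a mask, and the icon texture is sampled at those UVs.
# Alpha comes from the sampled RGB times AlphaGain and the mask, saturated.
# Color outputs the RGB premultiplied by that alpha.
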MaterialFunction /NiagaraFluids/Materials/Functions/Grid2D_DebugGizmos/ComputeIcon
graph {
 properties {bExposeToLibrary=true}
 nodes [
  n1 = FunctionInput(Preview=Constant() {r=8}) {bUsePreviewValueAsDefault=true, inputName=Size, inputType=FunctionInput_Scalar, previewValue=<8 8 8 8>}
  n2 = Constant2Vector()
  n3 = FunctionInput(Preview=Constant2Vector() {g=1, r=1}) {bUsePreviewValueAsDefault=true, inputName=Scale, inputType=FunctionInput_Scalar, previewValue=<8 8 8 8>}
  n4 = MaterialFunctionCall!(
   Size=@n1
   Offset=ComponentMask(None=FunctionInput(Preview=@n2) {bUsePreviewValueAsDefault=true, inputName=Offset, inputType=FunctionInput_Vector2}) {g=true, r=true}
   Scale=@n3
   CropLeft
   UV=FunctionInput(Preview=TextureCoordinate()) {bUsePreviewValueAsDefault=true, inputName=UV, inputType=FunctionInput_Vector2}
  ) {materialFunction=&/ComputeIconUV}
  n5 = TextureSample(
   UVs=@n4.UV
   Tex=FunctionInput(Preview=TextureObject() {texture=/NiagaraFluids/Textures/Grid2D_DebugGizmos/Grid2D_Nametag}) {bUsePreviewValueAsDefault=true, inputName=IconTexture, inputType=FunctionInput_Texture2D}
  ) {texture=&Grid2D_Nametag}
  n6 = Multiply(A=@n5.RGB, B=FunctionInput(Preview=Constant() {r=8}) {bUsePreviewValueAsDefault=true, inputName=AlphaGain, previewValue=<8 8 8 1>}) {constB=8}
  n7 = Saturate(None=Multiply(A=Multiply(A=@n6, B=@n4.Mask)))
  FunctionOutput(None=Multiply(A=@n5.RGB, B=@n7)) {outputName=Color}
  FunctionOutput(None=@n7) {outputName=Alpha}
 ]
}
