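# Unlit translucent Niagara sprite material that draws a debug gizmo for a Niagara Fluids 2D grid, with depth testing disabled.
# ResolutionX/Y scale the UVs into cell borders at four densities, blended by actor-to-camera distance so coarser lines replace finer ones farther away.
# A border texture, name tag and resolution labels are layered on top by lerps.
# ComputeDataIconInt renders both resolution values as numbers, positioned by offset parameters.
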Material /NiagaraFluids/Materials/Utility/DebugDisplayGrid2DGizmo
graph {
 properties {
  bDisableDepthTest=true
  bEnableResponsiveAA=true
  bUsedWithNiagaraSprites=true
  blendMode=BLEND_Translucent
  pixelDepthOffsetMode=PDOM_Legacy
  shadingModel=MSM_Unlit
  usedShadingModels=Unlit
 }
 nodes [
  Constant() {r=4}
  n1 = ScalarParameter() {defaultValue=512, parameterName=ResolutionX}
  n2 = ScalarParameter() {defaultValue=512, parameterName=ResolutionY}
  n3 = TextureCoordinate()
  n4 = Multiply(A=@n3, B=Constant2Vector() {g=1, r=1})
  n5 = Multiply(A=MaterialFunctionCall(X=@n1, Y=@n2) {materialFunction=&MakeFloat2}.Result, B=@n4)
  n6 = MaterialFunctionCall!(UV=Frac(None=@n5)) {materialFunction=/NiagaraFluids/Materials/Functions/Grid2D_DebugGizmos/ComputeBoxBorder}
  n7 = Constant() {r=0.4}
  n8 = MaterialFunctionCall(UV=Frac(None=Divide(A=@n5) {constB=8})) {materialFunction=&ComputeBoxBorder}
  n9 = MaterialFunctionCall(UV=Frac(None=Divide(A=@n5) {constB=16})) {materialFunction=&ComputeBoxBorder}
  n10 = MaterialFunctionCall(UV=Frac(None=Divide(A=@n5) {constB=32})) {materialFunction=&ComputeBoxBorder}
  n11 = ActorPositionWS()
  n12 = CameraPositionWS()
  n13 = Divide(A=MaterialFunctionCall(Vector 3=Subtract(A=@n11, B=@n12)) {materialFunction=/Engine/Functions/Engine_MaterialFunctions02/Utility/VectorLength}.V3 Length) {constB=1800}
  n14 = Multiply(A=TextureSample(UVs=@n3) {texture=/NiagaraFluids/Textures/Grid2D_DebugGizmos/Grid2D_Border}.RGB) {constB=1.5}
  n15 = LinearInterpolate(
   A=MaterialFunctionCall(
    1=Multiply(A=@n6.Result, B=@n7) {constB=0.25}
    2=Multiply(A=@n8.Result, B=@n7) {constB=0.25}
    3=Multiply(A=@n9.Result, B=@n7) {constB=0.25}
    4=Multiply(A=@n10.Result, B=@n7) {constB=0.25}
    A=Saturate(None=@n13)
   ) {materialFunction=&Lerp_Multiple_Float3}.Lerp 4 Inputs
   B=@n14
   Alpha=@n14
  )
  n16 = VectorParameter() {parameterName=TagOffset, defaultValue=<rgba 0.14 -0.25 0 1>}
  n17 = Constant2Vector() {g=-1, r=-1}
  n18 = MaterialFunctionCall!(
   Size=Constant() {r=6}
   Offset=MaterialFunctionCall(X=@n16.R, Y=@n16.R) {materialFunction=&MakeFloat2}.Result
   IconTexture=TextureObject() {texture=/NiagaraFluids/Textures/Grid2D_DebugGizmos/Grid2D_Nametag}
   AlphaGain
   Scale=@n17
   UV=@n4
  ) {materialFunction=/NiagaraFluids/Materials/Functions/Grid2D_DebugGizmos/ComputeIcon}
  n19 = VectorParameter() {parameterName=TagOffsetResolution, defaultValue=<rgba 0.18 0.12 0 1>}
  n20 = MaterialFunctionCall(
   Size=Constant() {r=6.5}
   Offset=MaterialFunctionCall(X=@n19.R, Y=@n19.R) {materialFunction=&MakeFloat2}.Result
   IconTexture=TextureObject() {texture=/NiagaraFluids/Textures/Grid2D_DebugGizmos/Grid2D_Nametag_Resolution}
   Scale=@n17
   UV=@n4
  ) {materialFunction=&ComputeIcon}
  n21 = VectorParameter() {parameterName=TagOffsetResolutionData, defaultValue=<rgba 1.9 0.8 0 1>}
  n22 = MaterialFunctionCall(
   Size=Constant() {r=8}
   Offset=MaterialFunctionCall(X=@n21.R, Y=@n21.R) {materialFunction=&MakeFloat2}.Result
   IconTexture=TextureObject() {texture=/NiagaraFluids/Textures/Grid2D_DebugGizmos/Grid2D_Nametag_Resolution_Data}
   Scale=Constant2Vector() {g=-1, r=-1.2}
   UV=@n4
  ) {materialFunction=&ComputeIcon}
  n23 = Constant() {r=8.5}
  n24 = VectorParameter() {parameterName=ResolutionXOffset, defaultValue=<rgba 1.62 -3.15 0 1>}
  n25 = Constant2Vector() {g=-0.5, r=-1}
  n26 = Constant() {r=0.5}
  n27 = VectorParameter() {parameterName=ResolutionYOffset, defaultValue=<rgba 0.65 0 0 1>}
  n28 = Add(
   A=MaterialFunctionCall!(Size=@n23, Offset=@n24.RGB, Scale=@n25, Number=@n1, CropLeft=@n26, UV=@n4) {materialFunction=/NiagaraFluids/Materials/Functions/Grid2D_DebugGizmos/ComputeDataIconInt}.Color
   B=MaterialFunctionCall(Size=@n23, Offset=Add(A=@n24.RGB, B=@n27.RGB), Scale=@n25, Number=@n2, CropLeft=@n26, UV=@n4) {materialFunction=&ComputeDataIconInt}.Color
  )
  LinearInterpolate(
   A=LinearInterpolate(A=LinearInterpolate(A=@n15, B=@n18.Color, Alpha=@n18.Color), B=@n20.Color, Alpha=@n20.Color)
   B=Add(A=@n22.Color, B=@n28)
   Alpha=Saturate(None=Add(A=@n22.Alpha, B=@n28))
  )
 ]
}
