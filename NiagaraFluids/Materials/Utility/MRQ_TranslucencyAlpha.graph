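# Post-process material in Niagara Fluids that derives a translucency alpha, named for use with Movie Render Queue.
# It samples PostProcessInput1 with filtering, isolates the alpha channel and inverts it with OneMinus.
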Material /NiagaraFluids/Materials/Utility/MRQ_TranslucencyAlpha
graph {
 properties {materialDomain=MD_PostProcess, pixelDepthOffsetMode=PDOM_Legacy, usedShadingModels=Default Lit}
 nodes [OneMinus(None=ComponentMask(None=SceneTexture() {bFiltered=true, sceneTextureId=PPI_PostProcessInput1}.Color) {a=true})]
}
