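# Translucent, unlit utility material in Niagara Fluids for solid particles, usable on Niagara sprite and mesh particles.
# A static switch, Use Sphere Mask, is on by default and multiplies particle color alpha by a hard-edged circular mask centered in UV space.
# With the switch off, the value is a constant one, ignoring both the mask and particle alpha.
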Material /NiagaraFluids/Materials/Utility/M_SolidParticle
graph {
 properties {
  bUsedWithNiagaraMeshParticles=true
  bUsedWithNiagaraSprites=true
  blendMode=BLEND_Translucent
  pixelDepthOffsetMode=PDOM_Legacy
  shadingModel=MSM_Unlit
  usedShadingModels=Unlit
 }
 nodes [
  n1 = TextureCoordinate()
  n2 = Constant() {r=0.5}
  n3 = Constant() {r=0.5}
  n4 = Constant() {r=1}
  StaticSwitchParameter(True=Multiply(A=ParticleColor().A, B=SphereMask(A=@n1, B=@n2, Radius=@n3, Hardness=@n4)), False=Constant() {r=1}) {defaultValue=true, parameterName=Use Sphere Mask}
 ]
}
